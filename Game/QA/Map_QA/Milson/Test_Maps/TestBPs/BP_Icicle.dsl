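; Destructible icicle test Blueprint that applies an icy feet effect to nearby characters when destroyed.
; On BeginPlay, random coin flips hide at most one of its two end pieces (Ice1 or Ice3) for visual variety.
; On server death, a sphere overlap sized by the Frost Explosion Radius scalable float finds actors, casts them to Character, and applies GE_PlayerIcyFeet from the icicle's own ability system component.
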(event EventBeginPlay
  ;[1 1 1] potentially hide one of the end icicles
  (if (Math|Random|RandomBool)
    (Development|SetHiddeninGame (Variables|Default|GetIce1) true false)
    (else
      (if (Math|Random|RandomBool)
        (Development|SetHiddeninGame (Variables|Default|GetIce3) true false)))))

(event EventOnDeathServer
  (params (Damage :type Float%20%28single-precision%29) (DamageTags :type Gameplay%20Tag%20Container%20Structure%20%28by%20ref%29) (Momentum :type Vector :default "0, 0, 0") (HitInfo :type Hit%20Result%20Structure%20%28by%20ref%29) (InstigatedBy :type Controller%20Object%20Reference) (DamageCauser :type Actor%20Object%20Reference) (EffectContext :type Gameplay%20Effect%20Context%20Handle%20Structure))
  ;[1 1 1] give the player icy feet if they are in range when the icicle is destroyed
  (bind _outactors (Collision|SphereOverlapActors (Transformation|GetActorLocation) (ScalableFloat|GetValueatLevel (Variables|Default|GetFrost%20Explosion%20Radius)) (Utilities|Array|MakeArray "ObjectTypeQuery3" "ObjectTypeQuery5")))
  (for _array_element _outactors
    (Utilities|Casting|CastToCharacter _array_element)
    (GameplayEffects|ApplyGameplayEffectToTarget :self (Ability|GetAbilitySystemComponent self) :GameplayEffectClass (Variables|Default|GetGE_PlayerIcyFeet) :Target (Ability|GetAbilitySystemComponent _array_element) :Level 0.0)))

(fn ConstructionScript)
